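# Animation Blueprint for the female BunnySuit body cosmetic, adding secondary motion to its ears, foot ears, grenade and tail.
# It copies the pose from the owner skeletal mesh found at initialization, then runs chained Anim Dynamics nodes, with ear alpha lowered while targeting.
# The result is cached, and when the playlist uses custom character parts, the output instead shrinks the head bone to near-zero scale.
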AnimBlueprint /BRCosmetics/Characters/Player/Female/Medium/Bodies/F_MED_BunnySuit_01/Meshes/F_MED_BunnySuit_01_Skeleton_AnimBP
target_skeleton=&/F_MED_BunnySuit_01_Skeleton
graphs [
 AnimGraph {nodes [
   n1 = "Transform (Modify) Bone - Bone: head"!(ComponentPose=Local To Component(LocalPose=Use cached pose 'PreScale Cache'()), Alpha=1) -> (Pose) {scaleMode=BMM_Replace, alphaBoolBlend {blendOption=Linear}, boneToModify {boneName=head}, scale=<0.01 0.01 0.01>}
   Output Pose(Result=Blend Poses by bool(
     BlendPose_0=Component To Local(ComponentPose=@n1)
     BlendPose_1=Use cached pose 'PreScale Cache'()
     BlendTime_0=0
     BlendTime_1=0
     bActiveValue=Get GetPlayerListUsesCustomCharacterParts() -> (GetPlayerListUsesCustomCharacterParts:bool)
    ) {blendType=Linear}) {layerGroup=None}
   n2 = Get MeshToCopy() -> (MeshToCopy:Skeletal Mesh Component Object Reference)
   n3 = SelectFloat(A=0.28, B=0.75, bPickA=Get bIsTargeting())
   n4 = Anim Dynamics - Bone: dyn_ear_01_r(ComponentPose=Local To Component(LocalPose=Copy Pose From Mesh(SourceMeshComponent=@n2)), Alpha=@n3, ExternalForce="0,0,0") {
    angularDampingOverride=0.96
    angularSpringConstant=0.98
    bAngularSpring=true
    bGravityOverrideInSimSpace=true
    bOverrideAngularDamping=true
    bOverrideLinearDamping=true
    bUseGravityOverride=true
    gravityScale=0.25
    linearDampingOverride=0.98
    alphaBoolBlend {blendOption=Linear}
    boundBone {boneName=dyn_ear_01_r}
    componentAppliedLinearAccClamp=<500 500 500>
    gravityOverride {z=-750}
    physicsBodyDefinitions [{
      sphereCollisionRadius=0
      boundBone {boneName=dyn_ear_01_r}
      boxExtents=<15 5 15>
      constraintSetup {angularLimitsMax=<0 0 15>, angularLimitsMin=<0 0 -15>, angularTarget=<0 0 0>}
      localJointOffset=<-5 0 0>
     }]
    relativeSpaceBone {boneName=spine_05}
   }
   n5 = Anim Dynamics - Bone: dyn_ear_02_r(ComponentPose=@n4, Alpha=@n3, ExternalForce="0,0,0") {
    angularDampingOverride=0.96
    angularSpringConstant=10
    bAngularSpring=true
    bGravityOverrideInSimSpace=true
    bOverrideAngularDamping=true
    bOverrideLinearDamping=true
    bUseGravityOverride=true
    gravityScale=0.25
    linearDampingOverride=0.98
    alphaBoolBlend {blendOption=Linear}
    boundBone {boneName=dyn_ear_02_r}
    componentAppliedLinearAccClamp=<500 500 500>
    gravityOverride {z=-750}
    physicsBodyDefinitions [{
      sphereCollisionRadius=0
      boundBone {boneName=dyn_ear_02_r}
      boxExtents=<10 5 10>
      constraintSetup {angularLimitsMax=<0 0 45>, angularLimitsMin=<0 0 -50>, angularTarget=<0 0 0>}
      localJointOffset=<-5 0 0>
     }]
    relativeSpaceBone {boneName=spine_05}
   }
   n6 = Anim Dynamics - Bone: dyn_ear_03_r(ComponentPose=@n5, Alpha=@n3, ExternalForce="0,0,0") {
    angularDampingOverride=0.999
    angularSpringConstant=20
    bAngularSpring=true
    bGravityOverrideInSimSpace=true
    bOverrideAngularDamping=true
    bOverrideLinearDamping=true
    bUseGravityOverride=true
    gravityScale=0.25
    alphaBoolBlend {blendOption=Linear}
    boundBone {boneName=dyn_ear_03_r}
    componentAppliedLinearAccClamp=<500 500 500>
    gravityOverride {z=-750}
    physicsBodyDefinitions [{
      sphereCollisionRadius=0
      boundBone {boneName=dyn_ear_03_r}
      boxExtents=<8 5 8>
      constraintSetup {angularLimitsMax=<5 0 45>, angularLimitsMin=<-5 0 -60>, angularTarget=<0 0 0>}
      localJointOffset=<-5 0 0>
     }]
    relativeSpaceBone {boneName=spine_05}
   }
   n7 = Anim Dynamics - Bone: dyn_ear_01_l(ComponentPose=@n6, Alpha=@n3, ExternalForce="0,0,0") {
    angularDampingOverride=0.96
    angularSpringConstant=0.98
    bAngularSpring=true
    bGravityOverrideInSimSpace=true
    bOverrideAngularDamping=true
    bOverrideLinearDamping=true
    bUseGravityOverride=true
    gravityScale=0.25
    linearDampingOverride=0.98
    alphaBoolBlend {blendOption=Linear}
    boundBone {boneName=dyn_ear_01_l}
    componentAppliedLinearAccClamp=<500 500 500>
    gravityOverride {z=-750}
    physicsBodyDefinitions [{
      sphereCollisionRadius=0
      boundBone {boneName=dyn_ear_01_l}
      boxExtents=<15 5 15>
      constraintSetup {angularLimitsMax=<0 0 15>, angularLimitsMin=<0 0 -15>, angularTarget=<0 0 0>}
      localJointOffset=<5 0 0>
     }]
    relativeSpaceBone {boneName=spine_05}
   }
   n8 = Anim Dynamics - Bone: dyn_ear_02_l(ComponentPose=@n7, Alpha=@n3, ExternalForce="0,0,0") {
    angularDampingOverride=0.96
    angularSpringConstant=10
    bAngularSpring=true
    bGravityOverrideInSimSpace=true
    bOverrideAngularDamping=true
    bOverrideLinearDamping=true
    bUseGravityOverride=true
    gravityScale=0.25
    linearDampingOverride=0.98
    alphaBoolBlend {blendOption=Linear}
    boundBone {boneName=dyn_ear_02_l}
    componentAppliedLinearAccClamp=<500 500 500>
    gravityOverride {z=-750}
    physicsBodyDefinitions [{
      sphereCollisionRadius=0
      boundBone {boneName=dyn_ear_02_l}
      boxExtents=<10 5 10>
      constraintSetup {angularLimitsMax=<0 0 45>, angularLimitsMin=<0 0 -50>, angularTarget=<0 0 0>}
      localJointOffset=<5 0 0>
     }]
    relativeSpaceBone {boneName=spine_05}
   }
   n9 = Anim Dynamics - Bone: dyn_ear_03_l(ComponentPose=@n8, Alpha=@n3, ExternalForce="0,0,0") {
    angularDampingOverride=0.999
    angularSpringConstant=20
    bAngularSpring=true
    bGravityOverrideInSimSpace=true
    bOverrideAngularDamping=true
    bOverrideLinearDamping=true
    bUseGravityOverride=true
    gravityScale=0.25
    alphaBoolBlend {blendOption=Linear}
    boundBone {boneName=dyn_ear_03_l}
    componentAppliedLinearAccClamp=<500 500 500>
    gravityOverride {z=-750}
    physicsBodyDefinitions [{
      sphereCollisionRadius=0
      boundBone {boneName=dyn_ear_03_l}
      boxExtents=<8 5 8>
      constraintSetup {angularLimitsMax=<5 0 45>, angularLimitsMin=<-5 0 -60>, angularTarget=<0 0 0>}
      localJointOffset=<5 0 0>
     }]
    relativeSpaceBone {boneName=spine_05}
   }
   n10 = Anim Dynamics - Bone: dyn_foot_inner_ear_r(ComponentPose=@n9, Alpha=0.5, ExternalForce="0,0,0") {
    angularSpringConstant=10
    bAngularSpring=true
    bGravityOverrideInSimSpace=true
    gravityScale=0.25
    simulationSpace=Actor
    alphaBoolBlend {blendOption=Linear}
    boundBone {boneName=dyn_foot_inner_ear_r}
    physicsBodyDefinitions [{
      sphereCollisionRadius=0
      boundBone {boneName=dyn_foot_inner_ear_r}
      boxExtents=<10 5 10>
      constraintSetup {angularLimitsMax=<0 0 30>, angularLimitsMin=<0 0 -30>, angularTarget=<0 0 0>}
      localJointOffset=<5 0 0>
     }]
   }
   n11 = Anim Dynamics - Bone: dyn_foot_outer_ear_r(ComponentPose=@n10, Alpha=0.5, ExternalForce="0,0,0") {
    angularSpringConstant=10
    bAngularSpring=true
    bGravityOverrideInSimSpace=true
    gravityScale=0.25
    simulationSpace=Actor
    alphaBoolBlend {blendOption=Linear}
    boundBone {boneName=dyn_foot_outer_ear_r}
    physicsBodyDefinitions [{
      sphereCollisionRadius=0
      boundBone {boneName=dyn_foot_outer_ear_r}
      boxExtents=<10 5 10>
      constraintSetup {angularLimitsMax=<0 0 30>, angularLimitsMin=<0 0 -30>, angularTarget=<0 0 0>}
      localJointOffset=<5 0 0>
     }]
   }
   n12 = Anim Dynamics - Bone: dyn_foot_inner_ear_l(ComponentPose=@n11, Alpha=0.5, ExternalForce="0,0,0") {
    angularSpringConstant=10
    bAngularSpring=true
    bGravityOverrideInSimSpace=true
    gravityScale=0.25
    simulationSpace=Actor
    alphaBoolBlend {blendOption=Linear}
    boundBone {boneName=dyn_foot_inner_ear_l}
    physicsBodyDefinitions [{
      sphereCollisionRadius=0
      boundBone {boneName=dyn_foot_inner_ear_l}
      boxExtents=<10 5 10>
      constraintSetup {angularLimitsMax=<0 0 30>, angularLimitsMin=<0 0 -30>, angularTarget=<0 0 0>}
      localJointOffset=<-5 0 0>
     }]
   }
   n13 = Anim Dynamics - Bone: dyn_foot_outer_ear_l(ComponentPose=@n12, Alpha=0.5, ExternalForce="0,0,0") {
    angularSpringConstant=10
    bAngularSpring=true
    bGravityOverrideInSimSpace=true
    gravityScale=0.25
    simulationSpace=Actor
    alphaBoolBlend {blendOption=Linear}
    boundBone {boneName=dyn_foot_outer_ear_l}
    physicsBodyDefinitions [{
      sphereCollisionRadius=0
      boundBone {boneName=dyn_foot_outer_ear_l}
      boxExtents=<10 5 10>
      constraintSetup {angularLimitsMax=<0 0 30>, angularLimitsMin=<0 0 -30>, angularTarget=<0 0 0>}
      localJointOffset=<-5 0 0>
     }]
   }
   n14 = Anim Dynamics - Bone: dyn_grenade(ComponentPose=@n13, Alpha=1, ExternalForce="0,0,0") {
    angularSpringConstant=1000
    bAngularSpring=true
    bGravityOverrideInSimSpace=true
    lODThreshold=1
    numSolverIterationsPostUpdate=4
    numSolverIterationsPreUpdate=16
    alphaBoolBlend {blendOption=Linear}
    boundBone {boneName=dyn_grenade}
    physicsBodyDefinitions [{
      sphereCollisionRadius=0
      boundBone {boneName=dyn_grenade}
      boxExtents=<20 10 10>
      constraintSetup {
       angularTargetAxis=AxisY
       twistAxis=AxisZ
       angularLimitsMax=<20 0 0>
       angularLimitsMin=<0 -20 0>
       angularTarget=<0 0 0>
       linearAxesMax=<0 0 0.5>
       linearAxesMin=<0 0 -0.5>
      }
      localJointOffset=<2 0 0>
     }]
    sphericalLimits [{limitRadius=16, drivingBone {boneName=pelvis}, sphereLocalOffset=<-1 1 0>}]
   }
   n15 = Anim Dynamics - Bone: dyn_tail(ComponentPose=@n14, Alpha=1, ExternalForce="0,0,0") {
    angularDampingOverride=0.5
    bGravityOverrideInSimSpace=true
    bOverrideAngularDamping=true
    alphaBoolBlend {blendOption=Linear}
    boundBone {boneName=dyn_tail}
    physicsBodyDefinitions [{
      sphereCollisionRadius=0
      boundBone {boneName=dyn_tail}
      boxExtents=<20 20 20>
      constraintSetup {angularLimitsMax=<0 30 20>, angularLimitsMin=<0 -30 -20>, angularTarget=<0 0 0>}
      localJointOffset=<-10 -10 0>
     }]
   }
   Save cached pose 'PreScale Cache'(Pose=Component To Local(ComponentPose=@n15)) {cacheName=PreScale Cache}
  ]}
 EventGraph {nodes [
   Event BlueprintInitializeAnimation()
   > Set MeshToCopy(MeshToCopy:Skeletal Mesh Component Object Reference=GetOwnerSkeletalMeshComponent()) -> (Output_Get:Skeletal Mesh Component Object Reference)
   > Set GetPlayerListUsesCustomCharacterParts(GetPlayerListUsesCustomCharacterParts:bool=GetPlaylistUsesCustomCharacterParts!() -> (ReturnValue:bool)) -> (Output_Get:bool)
   Event BlueprintUpdateAnimation()
  ]}
]
variables [
 GetPlayerListUsesCustomCharacterParts {type=boolean}
 MeshToCopy {type=SkeletalMeshComponent}
]
